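# Physics asset for the TagTiger vehicle body mesh in the VehicleCosmetics content.
# It holds a single root body with one box shape, offset forward and upward from the root bone to enclose the vehicle as a coarse collision volume.
# There are no constraints, so the body simulates as one rigid piece.
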PhysicsAsset /VehicleCosmetics/Bodies/TagTiger/Meshes/Body_TagTiger_PhysicsDM
bodies {root [{shapeType=Box, extentX=444, extentY=216, extentZ=144, shapeName=root_box, center=<18 0 110>}]}
constraints []
